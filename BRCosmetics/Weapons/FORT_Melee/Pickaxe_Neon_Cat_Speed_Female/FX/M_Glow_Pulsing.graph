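# Unlit translucent Niagara sprite and mesh-particle material for a scrolling, pulsing glow on the Neon Cat Speed female pickaxe.
# Particle color drives the emissive adjustment function.
# Panned water ramps, switchable between axes, combine with an offset gradient, procedural edge fade, view-angle fade and optional mirrored character trim mask, whose exponent is particle alpha.
# A separate time-based sine-cosine pulse has no visible link to that mask chain.
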Material /BRCosmetics/Weapons/FORT_Melee/Pickaxe_Neon_Cat_Speed_Female/FX/M_Glow_Pulsing
graph {
 properties {
  bUsedWithNiagaraMeshParticles=true
  bUsedWithNiagaraSprites=true
  blendMode=BLEND_Translucent
  pixelDepthOffsetMode=PDOM_Legacy
  shadingModel=MSM_Unlit
  usedShadingModels=Unlit
 }
 nodes [
  n1 = ParticleColor()
  MaterialFunctionCall(Emissive=@n1.RGB) {materialFunction=&MF_FX_EmissiveAdjustments}
  n2 = Sine(None=Multiply(A=Time()))
  n3 = Cosine(None=Multiply(A=Time()) {constB=0.05})
  Saturate(None=Add(A=Multiply(A=@n2, B=@n3)) {constB=1.5})
  n4 = TextureCoordinate()
  n5 = ComponentMask(None=@n4) {g=true}
  n6 = ScalarParameter() {parameterName=Offset_Mask}
  n7 = Multiply(A=Add(A=OneMinus(None=@n5), B=@n6) {constB=-0.1}, B=Add(A=@n5, B=@n6) {constB=-0.1})
  n8 = ComponentMask(None=@n4) {r=true}
  n9 = Multiply(A=Add(A=@n8, B=@n6) {constB=-0.1}, B=Add(A=OneMinus(None=@n8), B=@n6) {constB=-0.1})
  n10 = ScalarParameter() {defaultValue=1, parameterName=FadeHorizontalOrVertical?}
  n11 = TextureCoordinate()
  n12 = ComponentMask(None=@n11) {g=true}
  n13 = ComponentMask(None=@n11) {r=true}
  n14 = ScalarParameter() {parameterName=IsVerticalOrHorizontal?}
  n15 = Panner(
   Coordinate=Multiply(A=LinearInterpolate(A=@n12, B=@n13, Alpha=@n14))
   Speed=AppendVector(A=ScalarParameter() {defaultValue=1, parameterName=Speed_X}, B=ScalarParameter() {parameterName=Speed_Y})
  )
  n16 = TextureCoordinate()
  n17 = Add(
   A=LinearInterpolate(A=ComponentMask(None=@n16) {g=true}, B=ComponentMask(None=@n16) {r=true}, Alpha=@n14)
   B=ScalarParameter() {defaultValue=-0.2, parameterName=Offset_Y}
  ) {constB=-0.2}
  n18 = Multiply(
   A=LinearInterpolate(A=@n17, B=OneMinus(None=@n17), Alpha=ScalarParameter() {parameterName=InvertGradient?})
   B=ScalarParameter() {defaultValue=0.15, parameterName=Power}
  ) {constB=0.1}
  n19 = Multiply(
   A=LinearInterpolate(
    A=TextureSample(UVs=@n15) {texture=&/T_Water_Top_Ramp_90}.RGB
    B=TextureSample(UVs=@n15) {texture=/Game/Global/Textures/Default/Masks_and_Gradients/T_Water_Top_Ramp}.RGB
    Alpha=@n14
   )
   B=@n18
  )
  n20 = ScalarParameter() {defaultValue=20, parameterName=Power_Offset}
  n21 = Multiply(A=Saturate(None=LinearInterpolate(A=@n7, B=@n9, Alpha=@n10)), B=Add(A=Multiply(A=@n19, B=@n20) {constB=20}, B=Multiply(A=@n18)))
  n22 = TextureCoordinate()
  n23 = OneMinus(None=ComponentMask(None=@n22) {r=true})
  n24 = TextureSample(UVs=AppendVector(A=Add(A=@n23) {constB=0}, B=ComponentMask(None=@n22) {g=true})) {
   samplerSource=SSM_Clamp_WorldGroupSettings
   texture=/Game/Characters/Player/Female/Medium/Bodies/F_MED_Neon_Cat_Spy/Textures/T_F_MED_NeonCatSpy_Gradients_TrimDetail_Black
  }
  n25 = LinearInterpolate(B=Power(Base=@n24.RGB, Exp=@n1.A) {constExponent=0.5}, Alpha=ScalarParameter() {parameterName=Mask?}) {constA=1}
  n26 = ScalarParameter() {parameterName=FadeBasedOnViewAngle}
  Saturate(None=Multiply(
    A=Multiply(A=@n21, B=@n25)
    B=MaterialFunctionCall("FadeSensitivity (optional)"=@n26) {materialFunction=&FadeBasedOnViewAngle}.Result
   ))
 ]
}
